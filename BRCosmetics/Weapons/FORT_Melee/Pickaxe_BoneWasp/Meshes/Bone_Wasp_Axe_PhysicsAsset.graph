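# Physics asset for the Bone Wasp pickaxe, a two-handed melee weapon cosmetic, giving its skeletal mesh collision bodies and a jointed ring.
# A long capsule covers the handle bone, while boxes cover the dyn_ring bone and its dyn_ring_MSTR bone.
# A single constraint joins the two ring bodies with swing and twist limit angles, the tightest on the second swing axis.
# The handle body has no constraint to the ring bodies.
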PhysicsAsset /BRCosmetics/Weapons/FORT_Melee/Pickaxe_BoneWasp/Meshes/Bone_Wasp_Axe_PhysicsAsset
bodies {
 Melee_TwoHanded_Handle [{shapeType=Capsule, length=88.8457, radius=3.9146, center=<0.7757 -0.0868 25.1721>, rotation=<rot -0.0663 -90 -0.0144>}]
 dyn_ring [{shapeType=Box, extentX=8, extentY=2, extentZ=7, center=<3 0 0>, rotation=<rot 90 0 0>}]
 dyn_ring_MSTR [{shapeType=Box, extentX=2, extentY=2, extentZ=2}]
}
constraints [{bone1Name=dyn_ring, bone2Name=dyn_ring_MSTR, swing1LimitDegrees=30, swing2LimitDegrees=5, twistLimitDegrees=30}]
